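# Physics asset for the collision body in the cloth assets of the M_MED_PintPiano medium male player cosmetic body.
# It holds a single pelvis body with no collision shapes listed and no constraints.
# As written, it is effectively a placeholder, defining neither collision volumes nor a jointed ragdoll setup.
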PhysicsAsset /BRCosmetics/Characters/Player/Male/Medium/Bodies/M_MED_PintPiano/Meshes/ClothAssets/M_MED_PintPiano_MLD_Collision_Body_Physics
bodies {pelvis []}
constraints []
